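# Pickup sound cue for the HiInfantryRifle in the Figment Assault Infantry Athena plugin, reusing the heavy assault rifle pickup recording.
# A single wave player feeds a short fixed delay at the root, and the cue plays slightly pitched down at reduced volume.
# It uses the shared weapon pickup attenuation, the loot collect sound class and the PickupMaximum_Oldest concurrency set.
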SoundCue /Figment_Assault_Infantry_Athena/Audio/SC_HiInfantryRifle_Pickup_Cue
root=Delay_0
nodes [
 {id=Delay_0, children [WavePlayer_0], properties {delayMax=0.07, delayMin=0.07}}
 {id=WavePlayer_0, properties {soundWaveAssetPtr=/Game/Sounds/Fort_Weapons/AutoRifle/AssaultHeavy/AssaultHeavy_Pickup_01}}
]
properties {
 attenuationSettings=&Fort_Weapon_Pickup_Att
 maxDistance=3500
 pitchMultiplier=0.9
 soundClassObject=&Fort_Loot_Collect
 volumeMultiplier=0.45
 concurrencyOverrides {maxCount=1, resolutionRule=PreventNew}
 concurrencySet [&PickupMaximum_Oldest]
}
