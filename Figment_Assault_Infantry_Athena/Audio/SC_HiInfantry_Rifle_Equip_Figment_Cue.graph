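# Sound cue for equipping the Figment Infantry assault rifle.
# A short fixed delay feeds a single wave player that reuses the rifle's reload-end recording.
# It plays at reduced volume in the weapon foley sound class, using the shared weapon reload attenuation.
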SoundCue /Figment_Assault_Infantry_Athena/Audio/SC_HiInfantry_Rifle_Equip_Figment_Cue
root=Delay_0
nodes [
 {id=Delay_0, children [WavePlayer_1], properties {delayMax=0.1, delayMin=0.1}}
 {id=WavePlayer_1, properties {soundWaveAssetPtr=/Figment_Assault_Infantry_Athena/Audio/Wavs/HiInfantry_ReloadEnd_01}}
]
properties {
 attenuationSettings=&Fort_Weapon_Reload_Att
 maxDistance=2250
 soundClassObject=&Fort_WeaponFoley
 volumeMultiplier=0.5
}
