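# Sound cue for the first-person firing sound of the Figment Infantry assault rifle.
# A Random node picks one of three equally weighted close-range fire waves, and a Modulator adds slight variation with pitch floored at its original value.
# It plays at reduced volume through first-person projectile attenuation, a ranged-weapon 1P fire sound class and first-person gunfire concurrency.
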SoundCue /Figment_Assault_Infantry_Athena/Audio/SC_HiInfantry_Rifle_Fire_1p_Figment_Cue
root=Modulator_0
nodes [
 {id=Modulator_0, children [Random_1], properties {pitchMin=1}}
 {id=Random_1, children [WavePlayer_4,WavePlayer_5,WavePlayer_6], properties {weights [1,1,1]}}
 {id=WavePlayer_4, properties {soundWaveAssetPtr=/Figment_Assault_Infantry_Athena/Audio/Wavs/HiInfantry_Fire_Close_01}}
 {id=WavePlayer_5, properties {soundWaveAssetPtr=/Figment_Assault_Infantry_Athena/Audio/Wavs/HiInfantry_Fire_Close_03}}
 {id=WavePlayer_6, properties {soundWaveAssetPtr=/Figment_Assault_Infantry_Athena/Audio/Wavs/HiInfantry_Fire_Close_04}}
]
properties {
 attenuationSettings=&Fort_Weapon_Projectile_FirstPerson_Athena_Att
 maxDistance=2097152
 soundClassObject=&Fort_WeaponRanged_Fire_1P_PSM
 volumeMultiplier=0.43
 concurrencySet [&Weapon_Gunfire_FirstPerson_Conc]
}
