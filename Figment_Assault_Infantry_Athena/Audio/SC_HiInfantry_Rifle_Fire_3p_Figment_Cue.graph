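# Sound Cue for the third-person firing sound of the Figment HiInfantry assault rifle, layered by distance.
# Close, medium and distant layers each pick randomly from three waves, then pass through a modulator and their own attenuation.
# Sound class nodes route close shots to the 3P fire class and both farther layers to the distant class, whose sub-mix joins the close layer.
# Cue-level attenuation and concurrency reference first-person assets despite the 3p name.
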SoundCue /Figment_Assault_Infantry_Athena/Audio/SC_HiInfantry_Rifle_Fire_3p_Figment_Cue
root=Mixer_1
nodes [
 {id=Attenuation_11, children [Modulator_0], properties {attenuationSettings=&Fort_Weapon_Projectile_Close_Athena_Att}}
 {
  id=Attenuation_12
  children [Modulator_18]
  properties {attenuationSettings=&Fort_Weapon_Projectile_Medium_Mono_Athena_Att}
 }
 {
  id=Attenuation_13
  children [Modulator_19]
  properties {attenuationSettings=&Fort_Weapon_Projectile_Distant_Mono_Athena_Att}
 }
 {id=Mixer_0, children [SoundClass_16, SoundClass_15], properties {inputVolume [0.8,0.8]}}
 {id=Mixer_1, children [SoundClass_14, Mixer_0], properties {inputVolume [0.9,0.7]}}
 {id=Modulator_0, children [Random_2], properties {pitchMin=1}}
 {id=Modulator_18, children [Random_3], properties {volumeMax=1.7, volumeMin=1.7}}
 {id=Modulator_19, children [Random_4], properties {volumeMax=1.8, volumeMin=1.8}}
 {id=Random_2, children [WavePlayer_6,WavePlayer_7,WavePlayer_8], properties {weights [1,1,1]}}
 {id=Random_3, children [WavePlayer_9, WavePlayer_10, WavePlayer_11], properties {weights [1,1,1]}}
 {id=Random_4, children [WavePlayer_12, WavePlayer_13, WavePlayer_14], properties {weights [1,1,1]}}
 {
  id=SoundClass_14
  children [Attenuation_11]
  properties {soundClassOverride=&Fort_WeaponRanged_Fire_3P_PSM}
 }
 {
  id=SoundClass_15
  children [Attenuation_13]
  properties {soundClassOverride=&Fort_WeaponRanged_Fire_Distant}
 }
 {
  id=SoundClass_16
  children [Attenuation_12]
  properties {soundClassOverride=&Fort_WeaponRanged_Fire_Distant}
 }
 {id=WavePlayer_10, properties {soundWaveAssetPtr=/Figment_Assault_Infantry_Athena/Audio/Wavs/HiInfantry_Fire_Medium_02}}
 {id=WavePlayer_11, properties {soundWaveAssetPtr=/Figment_Assault_Infantry_Athena/Audio/Wavs/HiInfantry_Fire_Medium_03}}
 {id=WavePlayer_12, properties {soundWaveAssetPtr=/Figment_Assault_Infantry_Athena/Audio/Wavs/HiInfantry_Fire_Distant_03}}
 {id=WavePlayer_13, properties {soundWaveAssetPtr=/Figment_Assault_Infantry_Athena/Audio/Wavs/HiInfantry_Fire_Distant_02}}
 {id=WavePlayer_14, properties {soundWaveAssetPtr=/Figment_Assault_Infantry_Athena/Audio/Wavs/HiInfantry_Fire_Distant_01}}
 {id=WavePlayer_6, properties {soundWaveAssetPtr=/Figment_Assault_Infantry_Athena/Audio/Wavs/HiInfantry_Fire_Close_01}}
 {id=WavePlayer_7, properties {soundWaveAssetPtr=/Figment_Assault_Infantry_Athena/Audio/Wavs/HiInfantry_Fire_Close_03}}
 {id=WavePlayer_8, properties {soundWaveAssetPtr=/Figment_Assault_Infantry_Athena/Audio/Wavs/HiInfantry_Fire_Close_04}}
 {id=WavePlayer_9, properties {soundWaveAssetPtr=/Figment_Assault_Infantry_Athena/Audio/Wavs/HiInfantry_Fire_Medium_01}}
]
properties {
 attenuationSettings=&Fort_Weapon_Projectile_FirstPerson_Athena_Att
 maxDistance=2097152
 soundClassObject=&Fort_WeaponRanged_Fire_3P_PSM
 volumeMultiplier=0.45
 concurrencySet [&Weapon_Gunfire_FirstPerson_Conc]
}
