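# Sound cue for the start of the reload on the Figment heavy infantry assault rifle.
# A single wave player plays one reload-start wav directly as the cue root, with no randomization or modulation.
# It runs at reduced volume in the weapon foley sound class and uses the shared weapon reload attenuation for its falloff.
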SoundCue /Figment_Assault_Infantry_Athena/Audio/SC_HiInfantry_Rifle_Reload_Start_Figment_Cue
root=WavePlayer_1
nodes [{id=WavePlayer_1, properties {soundWaveAssetPtr=/Figment_Assault_Infantry_Athena/Audio/Wavs/HiInfantry_ReloadStart_01}}]
properties {
 attenuationSettings=&Fort_Weapon_Reload_Att
 maxDistance=2250
 soundClassObject=&Fort_WeaponFoley
 volumeMultiplier=0.5
}
