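# Sound cue for the Figment infantry rifle's ejected clip hitting the ground.
# A Random node picks one of four equally weighted M1 clip-eject ground-impact waves, and a Modulator raises it to a fixed higher pitch.
# The cue plays at reduced volume with weapon-reload attenuation and is routed to a ranged-weapon low-ammo sound class.
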SoundCue /Figment_Assault_Infantry_Athena/Audio/SC_Infantry_Rifle_ClipEject_HitGround_Figment_Cue
root=Modulator_0
nodes [
 {id=Modulator_0, children [Random_0], properties {pitchMax=1.2, pitchMin=1.2}}
 {id=Random_0, children [WavePlayer_0,WavePlayer_1,WavePlayer_2,WavePlayer_3], properties {weights [1,1,1,1]}}
 {id=WavePlayer_0, properties {soundWaveAssetPtr=&/M1_ClipEject_HitGround_01}}
 {id=WavePlayer_1, properties {soundWaveAssetPtr=&/M1_ClipEject_HitGround_03}}
 {id=WavePlayer_2, properties {soundWaveAssetPtr=&/M1_ClipEject_HitGround_04}}
 {id=WavePlayer_3, properties {soundWaveAssetPtr=&/M1_ClipEject_HitGround_05}}
]
properties {
 attenuationSettings=&Fort_Weapon_Reload_Att
 maxDistance=2250
 soundClassObject=/Game/Sounds/FortSoundClassesAndModes/Fort_WeaponRanged_LowAmmo_Degraded_PSM
 volumeMultiplier=0.7
}
