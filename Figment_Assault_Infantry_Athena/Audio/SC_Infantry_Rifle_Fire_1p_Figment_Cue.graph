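# First-person fire sound cue for the Infantry Rifle in the Figment Assault Infantry Athena content.
# A Random node picks one of four equally weighted M1 close-fire waves and feeds a Modulator for pitch and volume variation.
# The cue uses first-person weapon attenuation, a first-person ranged-fire sound class and a first-person gunfire concurrency set.
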SoundCue /Figment_Assault_Infantry_Athena/Audio/SC_Infantry_Rifle_Fire_1p_Figment_Cue
root=Modulator_0
nodes [
 {id=Modulator_0, children [Random_0], properties {pitchMin=1, volumeMax=1}}
 {id=Random_0, children [WavePlayer_0,WavePlayer_1,WavePlayer_2,WavePlayer_3], properties {weights [1,1,1,1]}}
 {id=WavePlayer_0, properties {soundWaveAssetPtr=&/M1_FireClose_01}}
 {id=WavePlayer_1, properties {soundWaveAssetPtr=&/M1_FireClose_02}}
 {id=WavePlayer_2, properties {soundWaveAssetPtr=&/M1_FireClose_03}}
 {id=WavePlayer_3, properties {soundWaveAssetPtr=&/M1_FireClose_04}}
]
properties {
 attenuationSettings=&Fort_Weapon_Projectile_FirstPerson_Athena_Att
 maxDistance=2097152
 soundClassObject=&Fort_WeaponRanged_Fire_1P_PSM
 volumeMultiplier=0.91
 concurrencySet [&Weapon_Gunfire_FirstPerson_Conc]
}
