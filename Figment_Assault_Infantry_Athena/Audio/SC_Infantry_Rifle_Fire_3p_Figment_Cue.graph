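# Sound cue for the third-person Infantry Rifle firing sound in Figment, built from M1 close, mid and distant recordings.
# Each layer picks a random wave, applies modulation and its own distance attenuation and sound class, then feeds a mixer weighted toward close.
# Mid and distant layers get roughly doubled modulator volume and exempt their random nodes from branch culling.
# Despite the 3p name, cue-level attenuation and concurrency use first-person assets.
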SoundCue /Figment_Assault_Infantry_Athena/Audio/SC_Infantry_Rifle_Fire_3p_Figment_Cue
root=Mixer_1
nodes [
 {id=Attenuation_11, children [Modulator_0], properties {attenuationSettings=&Fort_Weapon_Projectile_Close_Athena_Att}}
 {
  id=Attenuation_12
  children [Modulator_18]
  properties {attenuationSettings=&Fort_Weapon_Projectile_Medium_Mono_Athena_Att}
 }
 {
  id=Attenuation_13
  children [Modulator_19]
  properties {attenuationSettings=&Fort_Weapon_Projectile_Distant_Mono_Athena_Att}
 }
 {id=Mixer_1, children [SoundClass_14, SoundClass_16, SoundClass_15], properties {inputVolume [1,0.6,0.3]}}
 {id=Modulator_0, children [Random_0], properties {pitchMin=1, volumeMax=1}}
 {id=Modulator_18, children [Random_1], properties {volumeMax=2.05, volumeMin=1.95}}
 {id=Modulator_19, children [Random_17], properties {volumeMax=2.05, volumeMin=1.95}}
 {id=Random_0, children [WavePlayer_0,WavePlayer_1,WavePlayer_2,WavePlayer_3], properties {weights [1,1,1,1]}}
 {id=Random_1, children [WavePlayer_4,WavePlayer_5], properties {bShouldExcludeFromBranchCulling=true, weights [1,1]}}
 {id=Random_17, children [WavePlayer_38, WavePlayer_39], properties {bShouldExcludeFromBranchCulling=true, weights [1,1]}}
 {
  id=SoundClass_14
  children [Attenuation_11]
  properties {soundClassOverride=&Fort_WeaponRanged_Fire_3P_PSM}
 }
 {
  id=SoundClass_15
  children [Attenuation_13]
  properties {soundClassOverride=&Fort_WeaponRanged_Fire_Distant}
 }
 {
  id=SoundClass_16
  children [Attenuation_12]
  properties {soundClassOverride=&Fort_WeaponRanged_Fire_Distant}
 }
 {id=WavePlayer_0, properties {soundWaveAssetPtr=&/M1_FireClose_01}}
 {id=WavePlayer_1, properties {soundWaveAssetPtr=&/M1_FireClose_02}}
 {id=WavePlayer_2, properties {soundWaveAssetPtr=&/M1_FireClose_03}}
 {id=WavePlayer_3, properties {soundWaveAssetPtr=&/M1_FireClose_04}}
 {id=WavePlayer_38, properties {soundWaveAssetPtr=&/M1_FireDist_01}}
 {id=WavePlayer_39, properties {soundWaveAssetPtr=&/M1_FireDist_02}}
 {id=WavePlayer_4, properties {soundWaveAssetPtr=&/M1_FireMid_01}}
 {id=WavePlayer_5, properties {soundWaveAssetPtr=&/M1_FireMid_03}}
]
properties {
 attenuationSettings=&Fort_Weapon_Projectile_FirstPerson_Athena_Att
 maxDistance=2097152
 soundClassObject=&Fort_WeaponRanged_Fire_3P_PSM
 volumeMultiplier=0.91
 concurrencySet [&Weapon_Gunfire_FirstPerson_Conc]
}
