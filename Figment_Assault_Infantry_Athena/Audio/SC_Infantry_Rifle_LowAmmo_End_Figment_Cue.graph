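# Sound cue for the Figment Infantry Rifle's low-ammo end sound, playing an M1 clip eject sample.
# The wave player runs through a fixed short delay into a modulator that varies playback, with a raised pitch floor.
# The cue uses weapon reload attenuation and a low-ammo ranged weapon sound class, at slightly reduced volume.
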SoundCue /Figment_Assault_Infantry_Athena/Audio/SC_Infantry_Rifle_LowAmmo_End_Figment_Cue
root=Modulator_0
nodes [
 {id=Delay_0, children [WavePlayer_0], properties {delayMax=0.1, delayMin=0.1}}
 {id=Modulator_0, children [Delay_0], properties {pitchMin=1}}
 {id=WavePlayer_0, properties {soundWaveAssetPtr=&/M1_ClipEject_01}}
]
properties {
 attenuationSettings=&Fort_Weapon_Reload_Att
 maxDistance=2250
 soundClassObject=/Game/Sounds/FortSoundClassesAndModes/Fort_WeaponRanged_LowAmmo_Degraded_PSM
 volumeMultiplier=0.8
}
