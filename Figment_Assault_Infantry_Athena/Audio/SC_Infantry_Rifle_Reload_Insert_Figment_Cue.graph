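# Sound cue for the magazine or clip insert step of the Infantry Rifle reload in the Figment assault infantry weapon set.
# It plays a single wave, M1_ReloadInsert_02, straight from the root with no randomization or modulation.
# The cue uses the weapon reload attenuation and weapon foley sound class, with reduced volume.
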SoundCue /Figment_Assault_Infantry_Athena/Audio/SC_Infantry_Rifle_Reload_Insert_Figment_Cue
root=WavePlayer_0
nodes [{id=WavePlayer_0, properties {soundWaveAssetPtr=&/M1_ReloadInsert_02}}]
properties {
 attenuationSettings=&Fort_Weapon_Reload_Att
 maxDistance=2250
 soundClassObject=&Fort_WeaponFoley
 volumeMultiplier=0.6
}
